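; Quest ability for grapple glider distance that, once triggered by a launch event, tracks horizontal meters the player glides.
; A randomized looping timer checks pawn tags, resetting the reference point during skydive and accumulating distance while parachuting.
; Once meters pass a threshold, it sends a tagged GlideDistance quest stat and resets the count.
; It ends when the landing immunity tag is removed or no airborne tag remains, clearing the timer and discarding unsent meters.
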(event Ability|EventActivateAbilityFromEvent
  (params (EventData :type Gameplay%20Event%20Data%20Structure%20%28by%20ref%29))
  ;[1 1 1] Activate ability from event
  (|CallParentFunction EventData)
  (bind _asfort_player_pawn (Utilities|Casting|CastToFortPlayerPawn (Ability|GetAvatarActorfromActorInfo))
    (:then
      (bind _output_get (Variables|GAVars|Data|SetOwningPlayerPawn _asfort_player_pawn))
      ;;[1 1 1] We were just launched, now wait for this tag which is always applied on a launch to end the ability for us.
      (bind _asynctaskproxy (Ability|Tasks|WaitGameplayTagRemove "(TagName=\"Gameplay.Status.ImmuneFallingDamage.TillLand\")" 0 true)
        (:then
          (Variables|Default|SetLandedTrackTask _asynctaskproxy)
          (Utilities|IsValid (Fortnite|GetFortPlayerControllerfromActor _output_get)
            (:Is%20Valid
              (Variables|GAVars|Data|SetOwningPlayerQuestManager (Quests|GetQuestManager (Fortnite|GetFortPlayerControllerfromActor _output_get) "Athena"))
              (bind _output_get_2 (Variables|GAVars|Data|SetPreviousLocation (Transformation|GetActorLocation (Variables|GAVars|Data|GetOwningPlayerPawn))))
              (Variables|GAVars|Data|SetCurrentLocation _output_get_2)
              (bind _output_get_3 (Variables|GAVars|Setup|SetInterval (Math|Random|RandomFloatinRange 1.7 2.5)))
              (bind _returnvalue (Utilities|Time|SetTimerbyEvent (delegate Custom|SendDistanceUpdate) _output_get_3 true))
              (Variables|GAVars|Data|SetTImerHandle _returnvalue))
            (:Is%20Not%20Valid
              (Ability|EndAbility))))
        (:Removed
          (Ability|EndAbility))))
    (:CastFailed
      (Ability|EndAbility))))

(event Custom|SendDistanceUpdate
  (bind _t_stillhaventhitground (Variables|GAVars|Setup|GetT_StillHaventHitGround))
  (bind _t_parachute (Variables|GAVars|Setup|GetT_Parachute))
  (bind _t_skydive (Variables|GAVars|Setup|GetT_Skydive))
  (bind (_case_0_true _case_1_true _case_2_true) (GameplayTags|TagContainer|CompareTagAssetInterfacetoOtherTags _t_skydive _t_parachute _t_stillhaventhitground (Variables|GAVars|Data|GetOwningPlayerPawn)))
  (bind _owningplayerpawn (Variables|GAVars|Data|GetOwningPlayerPawn))
  (bind (_previouslocation_x _previouslocation_y) (Math|Vector|BreakVector (Variables|GetPreviousLocation_X)))
  (bind (_currentlocation_x _currentlocation_y) (Math|Vector|BreakVector (Variables|GetCurrentLocation_X)))
  ;[1 1 1] Send Distance Update
  (Variables|GAVars|Data|SetCurrentLocation (Transformation|GetActorLocation _owningplayerpawn))
  (if _case_0_true
    (Variables|GAVars|Data|SetPreviousLocation (Variables|GAVars|Data|GetCurrentLocation))
    (Variables|Default|SetPreviously%20Gliding true)
    (elif _case_1_true
      (if (Variables|Default|GetPreviously%20Gliding)
        ;;[1 1 1] Cache distance traveled (horizontal only)
        (Variables|GAVars|Data|SetPreviousLocation (Variables|GAVars|Data|GetCurrentLocation))
        (Variables|Default|SetPreviously%20Gliding false)
        (else
          _))
      (Variables|GAVars|Data|SetMetersTraveled (+ (Math|Float|Floor (/ (Math|Vector|VectorLength (Math|Vector|vector-vector (Math|Vector|MakeVector _currentlocation_x _currentlocation_y) (Math|Vector|MakeVector _previouslocation_x _previouslocation_y))) 1e+02)) (Variables|GAVars|Data|GetMetersTraveled)))
      (Variables|GAVars|Data|SetPreviousLocation (Variables|GAVars|Data|GetCurrentLocation)) ; Reset location
      ;;[1 1 1] Send Quest Progress
      (if (> (Variables|GAVars|Data|GetMetersTraveled) 299)
        (Quest|SendCustomStatEventwithTags (Variables|GAVars|Data|GetOwningPlayerQuestManager) "GlideDistance" (Variables|GAVars|Setup|GetT_Quest) (GameplayTags|GetOwnedGameplayTags (Variables|GAVars|Data|GetOwningPlayerPawn) :!class /Script/GameplayTags.GameplayTagAssetInterface) (Variables|GAVars|Data|GetMetersTraveled))
        (Variables|GAVars|Data|SetMetersTraveled 0))
      (elif _case_2_true
        _
        (else
          ;;[1 1 1 1 11] If we have hit the ground just end the ability. If we haven't just wait to enter a skydive or parachute state
          (Ability|EndAbility))))))

(event Custom|Final%20Distance%20Update
  ;[1 1 1] Send Distance Update
  ;[1 1 1] Did the player move? If So send final update
  (Variables|GAVars|Data|SetMetersTraveled 0))

(event Ability|EventOnEndAbility
  (params (bWasCancelled :type Boolean))
  (bind _timerhandle (Variables|GAVars|Data|GetTImerHandle))
  ;[1 1 1] On End Ability
  (|CallParentFunction bWasCancelled)
  (GameplayTasks|EndTask (Variables|Default|GetLandedTrackTask))
  (if (Utilities|Time|IsValidTimerHandle _timerhandle)
    (Utilities|Time|ClearandInvalidateTimerbyHandle _timerhandle)
    (else
      _))
  (CallFunction|FinalDistanceUpdate))
